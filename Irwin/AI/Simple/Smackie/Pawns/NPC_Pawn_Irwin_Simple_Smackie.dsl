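; NPC pawn Blueprint for the simple Smackie enemy or character, adding a landing sound.
; On BeginPlay it calls the parent, then binds the OnPawnLanded dispatcher to a handler that forwards the hit result to OnSmackieLanded.
; OnSmackieLanded plays the Smackie land sound at the actor's location, only where cosmetic events can execute.
; Overlap and Tick events are disabled stubs.
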(event EventBeginPlay
  (|CallParentFunction)
  (EventDispatchers|BindEventtoOnPawnLanded (delegate Custom|SmackieLanded)))

(event Collision|EventActorBeginOverlap
  (params (OtherActor :type Actor%20Object%20Reference))
  (|CallParentFunction OtherActor) ; This node is disabled and will not be called.\nDrag off pins to build functionality.
)

(event Custom|SmackieLanded
  (params (Hit :type Hit%20Result%20Structure%20%28by%20ref%29))
  (CallFunction|OnSmackieLanded Hit))

(event EventTick
  (params (DeltaSeconds :type Float%20%28single-precision%29))
  (|CallParentFunction DeltaSeconds) ; This node is disabled and will not be called.\nDrag off pins to build functionality.
)

(event Custom|OnSmackieLanded
  (params (HitResult :type Hit%20Result%20Structure))
  (Networking|CanExecuteCosmeticEvents
    (:True
      (Audio|PlaySoundatLocation (Variables|Default|GetSmackieLandSound) (Transformation|GetActorLocation)))))

(fn ConstructionScript
  (|CallParentFunction))
