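; Smackie frog NPC pawn variant that glows in the color of the best loot rarity in its inventory.
; When its cosmetic mesh loads or a pickup-grabbed cue fires (after a short delay), it looks up the rarity color in a data table.
; That color drives mesh emissive and a Niagara glow, active above Common, with spawn rate by rarity and scale from a curve table.
; On death it fades out the looping glow audio.
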(event EventBeginPlay
  ;[1 1 1] Do Frog Loot Glow
  (|CallParentFunction)
  (Networking|CanExecuteCosmeticEvents
    (:True
      (Customization|BindEventtoOnCustomizationsLoadedEvent (EventDispatchers|CreateEvent :!event Custom|MeshLoaded)) ; Track when Mesh gets loaded on Client
    )))

(event EventNPCSendGameplayCuePickupGrabbed
  (bind _npc_bestlootrarityininventory (Variables|NPCInventory|GetNPC%20BestLootRarityInInventory))
  (bind _color (CallFunction|GetLootGlowColorfromRarity _npc_bestlootrarityininventory))
  (bind _returnvalue (>= _npc_bestlootrarityininventory 1))
  (bind _returnvalue_2 (Math|Color|SelectColor :A _color :bPickA _returnvalue))
  (bind _returnvalue_3 (Math|Conversions|ToVector(LinearColor) _returnvalue_2))
  (bind _lootglow (Variables|Default|GetLootGlow))
  ;[1 1 1] Do Frog Loot Glow
  (Utilities|FlowControl|Delay 0.2) ; just in case
  (Utilities|FlowControl|Sequence
    (:then_0
      (Rendering|Material|SetVectorParameterValueonMaterials (Variables|Character|GetMesh) "EmissiveMultiplier" _returnvalue_3))
    (:then_1
      (Components|Activation|SetActive _lootglow _returnvalue false)
      (Effects|Components|ParticleSystem|SetVectorParameter _lootglow "CustomColor" _returnvalue_3)
      (Effects|Components|ParticleSystem|SetFloatParameter _lootglow "SpawnRateScale" (Utilities|Select :Index _npc_bestlootrarityininventory :Common 0.0 :Uncommon 0.0 :Rare 0.0 :Epic 0.0 :Legendary 0.5 :Mythic 1.0 :Transcendent 1.0 :Unattainable 1.0))
      (Effects|Components|ParticleSystem|SetFloatParameter _lootglow "LootEffectScale" (ScalableFloat|GetValueatLevel (struct :Value 1.0 :Curve (struct :CurveTable "/Irwin/DataTables/Irwin_Data.Irwin_Data" :RowName "Irwin.Loot.FX.DefaultLootEffectScale") :RegistryType "None") 1.0))
      (Niagara|ReinitializeSystem _lootglow))))

(event Custom|MeshLoaded
  (params (ThisPawn :type Fort%20Pawn%20Object%20Reference))
  (bind _npc_bestlootrarityininventory (Variables|NPCInventory|GetNPC%20BestLootRarityInInventory))
  (bind _color (CallFunction|GetLootGlowColorfromRarity _npc_bestlootrarityininventory))
  (bind _returnvalue (>= _npc_bestlootrarityininventory 1))
  (bind _returnvalue_2 (Math|Color|SelectColor :A _color :bPickA _returnvalue))
  (bind _returnvalue_3 (Math|Conversions|ToVector(LinearColor) _returnvalue_2))
  (bind _lootglow (Variables|Default|GetLootGlow))
  ;[1 1 1] Do Frog Loot Glow
  (Utilities|FlowControl|Sequence
    (:then_0
      (Rendering|Material|SetVectorParameterValueonMaterials (Variables|Character|GetMesh) "EmissiveMultiplier" _returnvalue_3))
    (:then_1
      (Components|Activation|SetActive _lootglow _returnvalue false)
      (Effects|Components|ParticleSystem|SetVectorParameter _lootglow "CustomColor" _returnvalue_3)
      (Effects|Components|ParticleSystem|SetFloatParameter _lootglow "SpawnRateScale" (Utilities|Select :Index _npc_bestlootrarityininventory :Common 0.0 :Uncommon 0.0 :Rare 0.0 :Epic 0.0 :Legendary 0.5 :Mythic 1.0 :Transcendent 1.0 :Unattainable 1.0))
      (Effects|Components|ParticleSystem|SetFloatParameter _lootglow "LootEffectScale" (ScalableFloat|GetValueatLevel (struct :Value 1.0 :Curve (struct :CurveTable "/Irwin/DataTables/Irwin_Data.Irwin_Data" :RowName "Irwin.Loot.FX.DefaultLootEffectScale") :RegistryType "None") 1.0))
      (Niagara|ReinitializeSystem _lootglow))))

(event EventOnDeathPlayEffects
  (params (Damage :type Float%20%28single-precision%29) (DamageTags :type Gameplay%20Tag%20Container%20Structure%20%28by%20ref%29) (Momentum :type Vector :default "0, 0, 0") (HitInfo :type Hit%20Result%20Structure%20%28by%20ref%29) (InstigatedBy :type Fort%20Pawn%20Object%20Reference) (DamageCauser :type Actor%20Object%20Reference) (EffectContext :type Gameplay%20Effect%20Context%20Handle%20Structure))
  (bind _lootglowaudioloop (Variables|Default|GetLootGlowAudioLoop))
  ;[1 1 1] Stop Loot Glow Looping Sound on Death
  (|CallParentFunction Damage DamageTags Momentum HitInfo InstigatedBy DamageCauser EffectContext)
  (Utilities|IsValid _lootglowaudioloop
    (:Is%20Valid
      (if (Audio|Components|Audio|IsPlaying _lootglowaudioloop)
        (Audio|Components|Audio|FadeOut _lootglowaudioloop 0.2 0.0 "Linear" :!class /Script/Engine.AudioComponent)))))

(fn GetLootGlowColorFromRarity
  (params (Rarity :type EFortRarity%20Enum :default "Common"))
  (bind _returnvalue (Utilities|GetDataTableRow "/Irwin/DataTables/LootGlowColor_Irwin.LootGlowColor_Irwin" (Utilities|Select :Index Rarity :Common "Common" :Uncommon "Uncommon" :Rare "Rare" :Epic "Epic" :Legendary "Legendary" :Mythic "Mythic" :Transcendent "Transcendent" :Unattainable "Unattainable"))
    (:then
      (return (Utilities|Struct|BreakLootGlowColor _returnvalue)))
    (:RowNotFound
      (return 0))))

(fn ConstructionScript
  (|CallParentFunction))
